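# FizzyCloud gadget sound cue for the cook sound, mixing a one-shot with a quieter looping layer.
# The one-shot randomly picks between two wave players, both using Cook_2-named waves, with pitch modulation.
# The loop layer starts after a short fixed delay, passing through an enveloper and a fixed pitch-down modulator.
# It uses weapon pickup attenuation, the gadgets sound class and its own concurrency limit.
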SoundCue /FizzyCloudGameplay/Assets/Audio/FizzyCloud_Cook_Cue
root=Mixer_0
nodes [
 {id=Delay_0, children [Enveloper_1], properties {delayMax=0.15, delayMin=0.15}}
 {
  id=Enveloper_1
  children [Modulator_1]
  properties {loopEnd=15, pitchCurve {editorCurveData {keys [{time=0}]}}, volumeCurve {editorCurveData {keys [<at 0 1>]}}}
 }
 {id=Mixer_0, children [Modulator_0,Delay_0], properties {inputVolume [1,0.25]}}
 {id=Modulator_0, children [Random_0], properties {pitchMin=0.9}}
 {id=Modulator_1, children [WavePlayer_3], properties {pitchMax=0.6, pitchMin=0.6}}
 {id=Random_0, children [WavePlayer_0,WavePlayer_2], properties {bShouldExcludeFromBranchCulling=true, weights [1,1]}}
 {id=WavePlayer_0, properties {soundWaveAssetPtr=&/FizzyCloud_Cook_2}}
 {id=WavePlayer_2, properties {soundWaveAssetPtr=&FizzyCloud_Cook_2}}
 {id=WavePlayer_3, properties {bLooping=true, soundWaveAssetPtr=&/FizzyCloud_Cook_Loop1}}
]
properties {
 attenuationSettings=&Fort_Weapon_Pickup_Att
 bOverrideConcurrency=true
 maxDistance=3500
 soundClassObject=&Fort_Gadgets
 volumeMultiplier=0.6
 concurrencyOverrides {maxCount=3}
}
