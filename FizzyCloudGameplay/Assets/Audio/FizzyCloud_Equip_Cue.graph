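# Sound cue for the FizzyCloud equip sound, using weapon reload attenuation and the loot collect sound class.
# A random node picks one of two equip wave variants with equal weight and feeds it through a root modulator.
# The cue plays at reduced volume and overrides concurrency to cap simultaneous instances.
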SoundCue /FizzyCloudGameplay/Assets/Audio/FizzyCloud_Equip_Cue
root=Modulator_0
nodes [
 {id=Modulator_0, children [Random_1]}
 {id=Random_1, children [WavePlayer_3,WavePlayer_4], properties {weights [1,1]}}
 {id=WavePlayer_3, properties {soundWaveAssetPtr=&/FizzyCloud_Equip_02}}
 {id=WavePlayer_4, properties {soundWaveAssetPtr=&/FizzyCloud_Equip_01}}
]
properties {
 attenuationSettings=&Fort_Weapon_Reload_Att
 bOverrideConcurrency=true
 maxDistance=2250
 soundClassObject=&Fort_Loot_Collect
 volumeMultiplier=0.3
 concurrencyOverrides {maxCount=3}
}
