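# Sound cue for the Fizzy Cloud landing sound in the FizzyCloudGameplay plugin.
# A Random node picks one of three equally weighted landing waves and feeds it through a default-setting Modulator to the output.
# The cue plays at reduced volume, borrowing weapon-reload attenuation and the loot-collect sound class.
# Its own concurrency override caps simultaneous instances at a small count.
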SoundCue /FizzyCloudGameplay/Assets/Audio/FizzyCloud_Land_Cue
root=Modulator_0
nodes [
 {id=Modulator_0, children [Random_2]}
 {id=Random_2, children [WavePlayer_0,WavePlayer_1,WavePlayer_2], properties {weights [1,1,1]}}
 {id=WavePlayer_0, properties {soundWaveAssetPtr=/FizzyCloudGameplay/Assets/Audio/waves/FizzyCloud_Land_03}}
 {id=WavePlayer_1, properties {soundWaveAssetPtr=/FizzyCloudGameplay/Assets/Audio/waves/FizzyCloud_Land_02}}
 {id=WavePlayer_2, properties {soundWaveAssetPtr=/FizzyCloudGameplay/Assets/Audio/waves/FizzyCloud_Land_01}}
]
properties {
 attenuationSettings=&Fort_Weapon_Reload_Att
 bOverrideConcurrency=true
 maxDistance=2250
 soundClassObject=&Fort_Loot_Collect
 volumeMultiplier=0.2
 concurrencyOverrides {maxCount=3}
}
