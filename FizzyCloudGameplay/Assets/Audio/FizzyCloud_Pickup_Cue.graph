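# FizzyCloud_Pickup_Cue is the SoundCue for picking up the FizzyCloud item.
# A Random node picks one of three equally weighted pickup waves and feeds it into a Modulator at the root.
# The cue plays at reduced volume through the loot-collect sound class with weapon-pickup attenuation, and a concurrency override caps how many instances play at once.
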SoundCue /FizzyCloudGameplay/Assets/Audio/FizzyCloud_Pickup_Cue
root=Modulator_0
nodes [
 {id=Modulator_0, children [Random_2]}
 {id=Random_2, children [WavePlayer_1,WavePlayer_3,WavePlayer_4], properties {weights [1,1,1]}}
 {id=WavePlayer_1, properties {soundWaveAssetPtr=/FizzyCloudGameplay/Assets/Audio/waves/FizzyCloud_PickUp_01}}
 {id=WavePlayer_3, properties {soundWaveAssetPtr=/FizzyCloudGameplay/Assets/Audio/waves/FizzyCloud_PickUp_02}}
 {id=WavePlayer_4, properties {soundWaveAssetPtr=/FizzyCloudGameplay/Assets/Audio/waves/FizzyCloud_PickUp_03}}
]
properties {
 attenuationSettings=&Fort_Weapon_Pickup_Att
 bOverrideConcurrency=true
 maxDistance=3500
 soundClassObject=&Fort_Loot_Collect
 volumeMultiplier=0.5
 concurrencyOverrides {maxCount=3}
}
