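# Looping Sound Cue for the Fizzy Cloud gadget's cloud ambience, in the Fort_Gadgets sound class and limited to four concurrent instances.
# A mixer blends inner, out and quad outer wave loops, each through its own pitch modulator, with the inner loop pitched sharply up.
# Each layer has custom distance attenuation with low-pass filtering and occlusion.
# The inner layer falls off sooner, both wider layers enable submix sends, and the quad layer adds listener focus.
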SoundCue /FizzyCloudGameplay/Assets/Audio/SC_FizzyCloud_Cloud_Loop
root=Modulator_1
nodes [
 {
  id=Attenuation_0
  children [WavePlayer_0]
  properties {
   bOverrideAttenuation=true
   attenuationOverrides {
    bAttenuateWithLPF=true
    bEnableLogFrequencyScaling=true
    bEnableOcclusion=true
    distanceAlgorithm=Custom
    falloffDistance=2500
    lPFFrequencyAtMax=100
    lPFRadiusMax=5000
    lPFRadiusMin=500
    nonSpatializedRadiusEnd=600
    nonSpatializedRadiusStart=1
    occlusionLowPassFilterFrequency=10000
    occlusionTraceChannel=ECC_GameTraceChannel11
    occlusionVolumeAttenuation=0.8
    customAttenuationCurve {editorCurveData {keys [
       <at 0 1 RCIM_Cubic>
       {arriveTangent=-5.8788, interpMode=RCIM_Cubic, leaveTangent=-5.8787, tangentMode=RCTM_User, time=0.0999, value=0.5175}
       {arriveTangent=-0.5827, interpMode=RCIM_Cubic, leaveTangent=-0.5827, tangentMode=RCTM_User, time=0.2091, value=0.2345}
       {arriveTangent=-0.2966, interpMode=RCIM_Cubic, leaveTangent=-0.2966, time=0.8588, value=0.1172}
       <at 1 0>
      ]}}
   }
  }
 }
 {
  id=Attenuation_1
  children [Modulator_3]
  properties {
   bOverrideAttenuation=true
   attenuationOverrides {
    bAttenuateWithLPF=true
    bEnableLogFrequencyScaling=true
    bEnableOcclusion=true
    bEnableSubmixSends=true
    distanceAlgorithm=Custom
    falloffDistance=6000
    lPFFrequencyAtMax=100
    lPFRadiusMax=5000
    lPFRadiusMin=500
    nonSpatializedRadiusEnd=800
    nonSpatializedRadiusStart=1
    occlusionLowPassFilterFrequency=10000
    occlusionTraceChannel=ECC_GameTraceChannel11
    reverbWetLevelMax=0.3
    reverbWetLevelMin=0.1
    stereoSpread=800
    customAttenuationCurve {editorCurveData {keys [
       <at 0 1 RCIM_Cubic>
       {arriveTangent=-5.8788, interpMode=RCIM_Cubic, leaveTangent=-5.8787, tangentMode=RCTM_User, time=0.0999, value=0.5175}
       {arriveTangent=-0.5827, interpMode=RCIM_Cubic, leaveTangent=-0.5827, tangentMode=RCTM_User, time=0.2092, value=0.2345}
       {arriveTangent=-0.2966, interpMode=RCIM_Cubic, leaveTangent=-0.2966, time=0.8588, value=0.1172}
       <at 1 0>
      ]}}
   }
  }
 }
 {
  id=Attenuation_2
  children [Modulator_2]
  properties {
   bOverrideAttenuation=true
   attenuationOverrides {
    bAttenuateWithLPF=true
    bEnableListenerFocus=true
    bEnableLogFrequencyScaling=true
    bEnableOcclusion=true
    bEnableSubmixSends=true
    distanceAlgorithm=Custom
    falloffDistance=6000
    lPFFrequencyAtMax=100
    lPFRadiusMax=5000
    lPFRadiusMin=500
    nonFocusVolumeAttenuation=0.8
    nonSpatializedRadiusEnd=800
    nonSpatializedRadiusStart=1
    occlusionLowPassFilterFrequency=10000
    occlusionTraceChannel=ECC_GameTraceChannel11
    reverbWetLevelMax=0.3
    reverbWetLevelMin=0.1
    stereoSpread=800
    attenuationShapeExtents {x=800}
    customAttenuationCurve {editorCurveData {keys [
       <at 0 1 RCIM_Cubic>
       {arriveTangent=-5.8788, interpMode=RCIM_Cubic, leaveTangent=-5.8787, tangentMode=RCTM_User, time=0.0999, value=0.5175}
       {arriveTangent=-0.5827, interpMode=RCIM_Cubic, leaveTangent=-0.5827, tangentMode=RCTM_User, time=0.2092, value=0.2345}
       {arriveTangent=-0.2966, interpMode=RCIM_Cubic, leaveTangent=-0.2966, time=0.8588, value=0.1172}
       <at 1 0>
      ]}}
   }
  }
 }
 {id=Mixer_0, children [Modulator_0, Attenuation_1, Attenuation_2], properties {inputVolume [0.8,0.6,0.7]}}
 {id=Modulator_0, children [Attenuation_0], properties {pitchMax=4, pitchMin=4}}
 {id=Modulator_1, children [Mixer_0], properties {pitchMax=1, pitchMin=1}}
 {id=Modulator_2, children [WavePlayer_1]}
 {id=Modulator_3, children [WavePlayer_2], properties {pitchMax=0.9, pitchMin=0.8}}
 {id=WavePlayer_0, properties {bLooping=true, soundWaveAssetPtr=/FizzyCloudGameplay/Assets/Audio/waves/FizzyCloud_Inner_Loop_01}}
 {id=WavePlayer_1, properties {bLooping=true, soundWaveAssetPtr=/FizzyCloudGameplay/Assets/Audio/waves/FizzyCloud_Outer_Loop_Quad}}
 {id=WavePlayer_2, properties {bLooping=true, soundWaveAssetPtr=/FizzyCloudGameplay/Assets/Audio/waves/FizzyCloud_Out_Loop_01}}
]
properties {
 bOverrideConcurrency=true
 maxDistance=6800
 soundClassObject=&Fort_Gadgets
 volumeMultiplier=0.5
 concurrencyOverrides {maxCount=4}
}
