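# Sound cue for the Fizzy Cloud explosion, routed to the gadgets sound class at reduced volume.
# A Random node picks one of three equally weighted explosion waves and feeds a root Modulator with default settings for playback variation.
# Spatialization references a dedicated explosion attenuation asset.
# The cue also stores inline attenuation values with custom falloff curves, low-pass, occlusion and focus, though no override flag is shown enabling them.
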SoundCue /FizzyCloudGameplay/Assets/Audio/SC_FizzyCloud_Explosion
root=Modulator_0
nodes [
 {id=Modulator_0, children [Random_0]}
 {id=Random_0, children [WavePlayer_0,WavePlayer_1,WavePlayer_2], properties {weights [1,1,1]}}
 {id=WavePlayer_0, properties {soundWaveAssetPtr=/FizzyCloudGameplay/Assets/Audio/waves/FizzyCloud_Explosion_01}}
 {id=WavePlayer_1, properties {soundWaveAssetPtr=/FizzyCloudGameplay/Assets/Audio/waves/FizzyCloud_Explosion_02}}
 {id=WavePlayer_2, properties {soundWaveAssetPtr=/FizzyCloudGameplay/Assets/Audio/waves/FizzyCloud_Explosion_03}}
]
properties {
 attenuationSettings=/FizzyCloudGameplay/GameplayCues/Assets/Audio/ATT_FizzyCloud_Explosion
 maxDistance=2097152
 soundClassObject=&Fort_Gadgets
 volumeMultiplier=0.4
 attenuationOverrides {
  absorptionMethod=CustomCurve
  bAttenuateWithLPF=true
  bEnableListenerFocus=true
  bEnableLogFrequencyScaling=true
  bEnableOcclusion=true
  distanceAlgorithm=Custom
  falloffDistance=7000
  lPFFrequencyAtMax=1000
  lPFRadiusMax=7000
  lPFRadiusMin=1500
  nonFocusVolumeAttenuation=0.6
  occlusionLowPassFilterFrequency=300
  occlusionTraceChannel=ECC_GameTraceChannel11
  reverbDistanceMax=1024
  reverbDistanceMin=0
  reverbWetLevelMax=0
  reverbWetLevelMin=1
  stereoSpread=5000
  attenuationShapeExtents {x=500}
  customAttenuationCurve {editorCurveData {keys [
     <at 0 1 RCIM_Cubic>
     {arriveTangent=-1.7545, interpMode=RCIM_Cubic, leaveTangent=-1.7545, tangentMode=RCTM_User, time=0.1872, value=0.8966}
     {arriveTangent=-2.2893, interpMode=RCIM_Cubic, leaveTangent=-2.2893, tangentMode=RCTM_User, time=0.2916, value=0.4221}
     {arriveTangent=-0.3694, interpMode=RCIM_Cubic, leaveTangent=-0.3694, tangentMode=RCTM_User, time=0.5162, value=0.1536}
     {interpMode=RCIM_Cubic, leaveTangent=-0.774, time=0.961, value=0.0513}
    ]}}
  customHighpassAirAbsorptionCurve {editorCurveData {keys [
     <at 0 1 RCIM_Cubic>
     {arriveTangent=-1.7545, interpMode=RCIM_Cubic, leaveTangent=-1.7545, tangentMode=RCTM_User, time=0.1872, value=0.8966}
     {arriveTangent=-2.2893, interpMode=RCIM_Cubic, leaveTangent=-2.2893, tangentMode=RCTM_User, time=0.2916, value=0.4221}
     {arriveTangent=-0.3694, interpMode=RCIM_Cubic, leaveTangent=-0.3694, tangentMode=RCTM_User, time=0.5162, value=0.1536}
     {interpMode=RCIM_Cubic, leaveTangent=-0.774, time=0.961, value=0.0513}
    ]}}
  customLowpassAirAbsorptionCurve {editorCurveData {keys [
     <at 0 1 RCIM_Cubic>
     {arriveTangent=-1.7545, interpMode=RCIM_Cubic, leaveTangent=-1.7545, tangentMode=RCTM_User, time=0.1872, value=0.8966}
     {arriveTangent=-2.2893, interpMode=RCIM_Cubic, leaveTangent=-2.2893, tangentMode=RCTM_User, time=0.2916, value=0.4221}
     {arriveTangent=-0.3694, interpMode=RCIM_Cubic, leaveTangent=-0.3694, tangentMode=RCTM_User, time=0.5162, value=0.1536}
     {interpMode=RCIM_Cubic, leaveTangent=-0.774, time=0.961, value=0.0513}
    ]}}
 }
}
