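# Sound cue for the Fizzy Cloud gadget's in-air loop, routed to the gadgets sound class with its own attenuation.
# A mixer layers a start one-shot with a looping wave that is briefly delayed and shaped by an enveloper volume curve.
# Both layers are pitched down, the loop at a fixed amount and the start with slight random variation, before a root modulator.
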SoundCue /FizzyCloudGameplay/Assets/Audio/SC_FizzyCloud_InAir_Loop
root=Modulator_0
nodes [
 {id=Delay_2, children [Enveloper_0], properties {delayMax=0.25, delayMin=0.25}}
 {
  id=Enveloper_0
  children [Modulator_1]
  properties {
   loopEnd=0.3
   loopStart=0.3
   pitchCurve {editorCurveData {keys [{time=0}]}}
   volumeCurve {editorCurveData {keys [<at 0.2178 0.9752>,<at 1 1>]}}
  }
 }
 {id=Mixer_0, children [Delay_2,Modulator_2], properties {inputVolume [0.3,0.2]}}
 {id=Modulator_0, children [Mixer_0]}
 {id=Modulator_1, children [WavePlayer_1], properties {pitchMax=0.6, pitchMin=0.6}}
 {id=Modulator_2, children [WavePlayer_2], properties {pitchMax=0.76, pitchMin=0.7}}
 {id=WavePlayer_1, properties {bLooping=true, soundWaveAssetPtr=/FizzyCloudGameplay/Assets/Audio/waves/FizzyCloud_InAirLoop}}
 {id=WavePlayer_2, properties {soundWaveAssetPtr=/FizzyCloudGameplay/Assets/Audio/waves/FizzyCloud_Loop_Start_01}}
]
properties {
 attenuationSettings=&/InAirloop_Att
 maxDistance=12400
 soundClassObject=&Fort_Gadgets
 volumeMultiplier=1
}
