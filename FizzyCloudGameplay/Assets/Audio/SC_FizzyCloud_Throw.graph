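# Sound cue for the Fizzy Cloud gadget throw, in the Fort_Gadgets sound class.
# A mixer combines a modulated in-air loop start wave with a briefly delayed, pitched-up throw wave, but the throw input's volume is zero, silencing that layer.
# Overridden attenuation, used in place of the referenced gadget throw asset, adds distance filtering, occlusion and a distant reverb submix send.
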SoundCue /FizzyCloudGameplay/Assets/Audio/SC_FizzyCloud_Throw
root=Mixer_0
nodes [
 {id=Delay_0, children [Modulator_2], properties {delayMax=0.07, delayMin=0.07}}
 {id=Mixer_0, children [Delay_0,Modulator_1], properties {inputVolume [0,1]}}
 {id=Modulator_1, children [WavePlayer_0]}
 {id=Modulator_2, children [WavePlayer_2], properties {pitchMax=1.25, pitchMin=1.2}}
 {id=WavePlayer_0, properties {soundWaveAssetPtr=/FizzyCloudGameplay/Assets/Audio/waves/FizzyCloud_InAirLoop_Start_01}}
 {id=WavePlayer_2, properties {soundWaveAssetPtr=/FizzyCloudGameplay/Assets/Audio/waves/FizzyCloud_Throw_01}}
]
properties {
 attenuationSettings=/Game/Sounds/Attenuation/Fort_GadgetThrow_Att
 bOverrideAttenuation=true
 maxDistance=1900
 soundClassObject=&Fort_Gadgets
 volumeMultiplier=0.6
 attenuationOverrides {
  bAttenuateWithLPF=true
  bEnableOcclusion=true
  bEnableSubmixSends=true
  falloffDistance=1500
  hPFFrequencyAtMax=125
  lPFFrequencyAtMax=1200
  lPFRadiusMax=150
  lPFRadiusMin=300
  nonSpatializedRadiusMode=Direct2D
  occlusionLowPassFilterFrequency=5000
  occlusionTraceChannel=ECC_GameTraceChannel11
  occlusionVolumeAttenuation=1.5
  reverbDistanceMax=3000
  reverbDistanceMin=0
  reverbWetLevelMax=1
  reverbWetLevelMin=0.25
  stereoSpread=500
  customAttenuationCurve {editorCurveData {keys [
     <at 0 1 RCIM_Cubic>
     <at 0.2336 0.8799>
     {arriveTangent=-1.148, interpMode=RCIM_Cubic, leaveTangent=-1.148, time=0.2741, value=0.3894}
     {arriveTangent=-0.7405, interpMode=RCIM_Cubic, leaveTangent=-0.7405, tangentMode=RCTM_User, time=1, value=0}
    ]}}
  submixSendSettings [{
    bEnableHPFCutoff=false
    bEnableLPFCutoff=false
    disableManualSendClamp=false
    hPFCutoff=20
    lPFCutoff=20000
    maxSendDistance=1500
    maxSendLevel=0.2
    minSendDistance=0
    minSendLevel=0
    sendLevel=0.2
    sendLevelControlMethod=Linear
    soundSubmix=&Fortnite_Reverb_Distant
   }]
 }
}
