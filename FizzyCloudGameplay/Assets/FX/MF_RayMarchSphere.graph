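# Material function that gives one depth slice of a sphere-like volume, sampling noise from a 2D slice atlas for FX materials, likely clouds.
# It intersects the view ray with a camera-facing plane through the object center, shifted toward the camera by a scalar input, without looping.
# That point, adjusted by scale and local offset, drives MF_3DVolumeFrom2DSlices.
# Outputs are remapped RGB, alpha as Mask, and a scale-relative SphericalNormal.
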MaterialFunction /FizzyCloudGameplay/Assets/FX/MF_RayMarchSphere
graph {nodes [
  ActorPositionWS() {originType=CameraRelative}
  ActorPositionWS() {originType=CameraRelative}
  n1 = Subtract(A=WorldPosition() {worldPositionShaderOffset=WPT_CameraRelative}.XYZ, B=ObjectPositionWS() {originType=CameraRelative})
  n2 = Normalize(VectorInput=ObjectPositionWS() {originType=CameraRelative})
  n3 = CameraVectorWS()
  n4 = Divide(A=DotProduct(A=@n1, B=@n2), B=DotProduct(A=@n2, B=@n3))
  n5 = Constant()
  n6 = Add(
   A=Subtract(A=@n1, B=Multiply(A=@n4, B=@n3))
   B=Multiply(A=Multiply(A=@n2) {constB=-1}, B=FunctionInput(Preview=@n5) {bUsePreviewValueAsDefault=true, inputType=FunctionInput_Scalar}) {constB=1.2}
  )
  n7 = Multiply(A=@n6) {constB=0.01}
  n8 = FunctionInput() {bUsePreviewValueAsDefault=true, inputName=Scale, inputType=FunctionInput_Scalar, previewValue=<1 0 0 1>}
  n9 = Multiply(A=@n8, B=MaterialFunctionCall() {materialFunction=&ObjectScale}.Scale XYZ)
  n10 = Constant3Vector() {constant=<rgba 0 0 0 1>}
  n11 = MaterialFunctionCall!(
   UVs=Add(
    A=Divide(A=@n7, B=@n9)
    B=FunctionInput(Preview=@n10) {bUsePreviewValueAsDefault=true, description=Local Space offset, inputName=ObjectPositionOffset, sortPriority=1}
   )
   Slices Texture=FunctionInput(Preview=TextureObject() {samplerType=SAMPLERTYPE_LinearColor, texture=&/T_3DNoise_N}) {inputName=VolumeTexture, inputType=FunctionInput_Texture2D}
   Atlas Size=Constant() {r=8}
   Atlas Padding=ScalarParameter() {defaultValue=0.9375, parameterName=AtlasPadding, primitiveDataIndex=8}
  ) {materialFunction=/Game/Environments/Helios/TechArt/LowAltitudeClouds/MF_3DVolumeFrom2DSlices}
  FunctionOutput(None=ConstantBiasScale(None=@n11.RGB) {bias=-0.5, scale=2}) {outputName=RGB}
  n12 = Constant() {r=50}
  n13 = MaterialFunctionCall() {materialFunction=&ObjectScale}
  FunctionOutput(None=Divide(A=@n6, B=Multiply(A=@n12, B=@n13.Scale XYZ))) {outputName=SphericalNormal}
  FunctionOutput(None=@n11.A) {outputName=Mask}
 ]}
