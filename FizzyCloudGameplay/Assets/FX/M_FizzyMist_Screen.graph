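# Translucent unlit material for a fizzy mist effect, apparently screen-covering, drawn after motion blur without depth test and usable on Niagara mesh particles.
# Animated 3D noise, sampled from a 2D slice atlas using view direction and camera position, yields sky-light-lit colour with time-of-day tinting and a contrast-shaped opacity.
# Camera-distance and timestamp transition fades are computed, some via vertex interpolators.
# Dangling nodes include a scene-colour blend and debug readout.
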Material /FizzyCloudGameplay/Assets/FX/M_FizzyMist_Screen
graph {
 properties {
  bDisableDepthTest=true
  bUseTranslucencyVertexFog=false
  bUsedWithNiagaraMeshParticles=true
  blendMode=BLEND_Translucent
  blendableLocation=BL_SceneColorBeforeBloom
  pixelDepthOffsetMode=PDOM_Legacy
  shadingModel=MSM_Unlit
  translucencyPass=MTP_AfterMotionBlur
  usedShadingModels=Unlit
 }
 nodes [
  n1 = Multiply(A=Multiply(A=CameraVectorWS()) {constB=-1}, B=ScalarParameter() {defaultValue=0.75, parameterName=PP_NoiseScale}) {constB=0.01}
  n2 = Time()
  n3 = Multiply(
   A=Multiply(A=Constant3Vector() {constant=<rgba 1 1 -1 1>}, B=ScalarParameter() {defaultValue=1.5, parameterName=PP_NoiseSpeed})
   B=Fmod(A=Multiply(A=@n2) {constB=0.05}, B=Constant() {r=1000})
  )
  Add(A=@n1, B=@n3)
  n4 = NamedRerouteUsage()
  n5 = Multiply(A=CameraPositionWS()) {constB=-0.0007}
  n6 = MaterialFunctionCall!(
   UVs=Add(A=Subtract(A=@n1, B=@n5), B=@n3)
   Slices Texture=TextureObject() {samplerType=SAMPLERTYPE_LinearColor, texture=&/T_3DNoise_N}
   Atlas Size=Constant() {r=8}
   Atlas Padding=ScalarParameter() {defaultValue=0.9404, parameterName=AtlasPadding, primitiveDataIndex=8}
  ) {materialFunction=/Game/Environments/Helios/TechArt/LowAltitudeClouds/MF_3DVolumeFrom2DSlices}
  n7 = Add(A=@n6.RGB) {constB=-0.5}
  n8 = ConstantBiasScale(None=DotProduct(A=Multiply(A=@n7) {constB=2}, B=SkyAtmosphereLightDirection()))
  n9 = VectorParameter() {parameterName=Color Multiplier, primitiveDataIndex=3, channelNames=<rgba 0 0.781 0.348 1>, defaultValue=<rgba 0.6877 0.201 3 1>}
  n10 = VectorParameter() {parameterName=Color Multiplier Core, primitiveDataIndex=3, channelNames=<rgba 0 0.781 0.348 1>, defaultValue=<rgba 0.3542 0.6798 1 1>}
  n11 = Multiply(A=LinearInterpolate(Alpha=Multiply(A=@n8, B=@n8)) {constA=0.1}, B=LinearInterpolate(A=@n9.RGB, B=@n10.RGB, Alpha=NamedRerouteUsage()))
  n12 = Multiply(A=@n11, B=ScalarParameter() {defaultValue=1, parameterName=ToD_NightMultiplier, primitiveDataIndex=2})
  n13 = VectorParameter() {parameterName=ToD_EveningMultiplier, primitiveDataIndex=8, defaultValue=<rgba 1 0.4101 0.3886 1>}
  n14 = VectorParameter() {parameterName=ToD_NightMultiplier, primitiveDataIndex=12, defaultValue=<rgba 0.3529 0.538 1 1>}
  n15 = Multiply(
   A=MaterialFunctionCall(Color_Morning=@n11, Color_Day=@n11, Color_Evening=Multiply(A=@n12, B=@n13.RGB), Color_Night=Multiply(A=@n12, B=@n14.RGB)) {materialFunction=&Color_Time_of_Day}.Result
   B=EyeAdaptationInverse().EyeAdaptationInverse
  )
  Add(A=@n15, B=@n4)
  n16 = VectorParameter() {parameterName=DensityContrast_RG, defaultValue=<rgba 3.9392 7.5419 -2.1513 -3.0102>}
  n17 = NamedRerouteUsage()
  n18 = NamedRerouteUsage()
  n19 = Saturate(None=Add(A=LinearInterpolate(A=@n16.R, B=@n16.R, Alpha=@n17) {constA=0.5}, B=LinearInterpolate(A=@n16.B, B=@n16.B, Alpha=@n18)))
  n20 = ScalarParameter() {defaultValue=1, parameterName=Opacity}
  n21 = NamedRerouteUsage()
  n22 = Saturate(None=Multiply(A=@n19, B=Multiply(A=@n20, B=@n21)))
  Add(A=@n4, B=@n22)
  Constant() {r=1}
  LinearInterpolate(A=TextureCoordinate(), B=TextureCoordinate() {coordinateIndex=1}, Alpha=ScalarParameter() {parameterName=CubeMap_Switch})
  LinearInterpolate(
   A=ComponentMask(None=SceneTexture() {bFiltered=true, sceneTextureId=PPI_PostProcessInput0}.Color) {b=true, g=true, r=true}
   B=@n15
   Alpha=@n22
  )
  n23 = Subtract(A=Multiply(A=ScreenPosition().ViewportUV) {constB=2})
  n24 = Multiply(A=@n23, B=@n23) {constB=2}
  n25 = MaterialFunctionCall(Float2=OneMinus(None=Abs(None=@n24))) {materialFunction=&BreakOutFloat2Components}
  n26 = Multiply(A=@n25.R, B=@n25.R)
  MaterialFunctionCall(A=@n25.R, B=@n26, C=@n25.R, Alpha=ScalarParameter() {defaultValue=0.5, parameterName=CubeMap_Select}) {materialFunction=&Lerp_3Color}
  n27 = CameraPositionWS()
  n28 = VectorParameter() {parameterName=ActorLocation, defaultValue=<rgba 0 0 0 1>}
  n29 = Subtract(A=Subtract(A=Distance(A=@n27, B=@n28.RGB)) {constB=500}, B=ScalarParameter() {defaultValue=400, parameterName=Offset, primitiveDataIndex=18})
  n30 = Saturate(None=Divide(A=Multiply(A=@n29) {constB=-1}, B=ScalarParameter() {defaultValue=32, parameterName=Fade, primitiveDataIndex=18}))
  Multiply(A=@n19, B=@n30)
  NamedRerouteDeclaration(None=@n6.A) {name=Mask, nodeColor=<rgba 0 0.0156 1 1>}
  n31 = CameraPositionWS()
  n32 = ObjectPositionWS()
  n33 = Normalize(VectorInput=Subtract(A=@n31, B=@n32))
  n34 = Transform(None=Constant4Vector() {constant=<rgba 0 0 1 0>}.RGB) {transformSourceType=TRANSFORMSOURCE_Camera}
  n35 = Add(
   A=Length(None=Subtract(A=@n31, B=@n32)) {is=class MaterialExpressionLength}
   B=Multiply(A=DotProduct(A=@n33, B=@n34), B=ScalarParameter() {defaultValue=250, parameterName=CameraDirectionOffset})
  )
  n36 = VectorParameter() {parameterName=CameraDistanceRemap, defaultValue=<rgba 500 800 1 0>}
  n37 = MaterialFunctionCall(Input=@n35, Input Low=@n36.R, Input High=@n36.R, Target Low=@n36.R, Target High=@n36.R) {materialFunction=&RemapValueRange}
  NamedRerouteDeclaration(None=VertexInterpolator(VS=Saturate(None=@n37.Result)).PS) {name=CameraFadeValue, nodeColor=<rgba 0 0.3672 1 1>}
  n38 = ScalarParameter() {defaultValue=512, parameterName=512}
  NamedRerouteDeclaration(None=MaterialFunctionCall(
    Number=@n35
    UVs=MaterialFunctionCall(Texture Dimensions=@n38) {materialFunction=&ScreenAlignedUVs}."X 100%, Y 100%"
   ) {materialFunction=&DebugScalarValues}.Result) {name=debug, nodeColor=<rgba 0.4766 1 0 1>}
  NamedRerouteDeclaration(None=@n26) {name=CubeMap_2, nodeColor=<rgba 1 1 1 1>}
  n39 = Subtract(A=Time(), B=ScalarParameter() {parameterName=InTimestamp})
  n40 = ScalarParameter() {defaultValue=1.5, desc=InLength, parameterName=TransitionLength, primitiveDataIndex=16}
  n41 = NamedRerouteDeclaration(None=Clamp(None=Divide(A=@n39, B=@n40) {desc=InLength})) {name=InTimestamp, nodeColor=<rgba 1 0 0.0469 1>}
  n42 = LinearInterpolate(A=@n41, B=OneMinus(None=@n41), Alpha=ScalarParameter() {parameterName=InvertTimestamp, primitiveDataIndex=1})
  n43 = ScalarParameter() {parameterName=SmokeDebug, primitiveDataIndex=7}
  NamedRerouteDeclaration(None=VertexInterpolator(VS=LinearInterpolate(A=@n42, Alpha=@n43)).PS) {name=Timestamp, nodeColor=<rgba 0 0.6953 1 1>}
  Reroute(None=Multiply(A=Saturate(None=@n29), B=Constant3Vector() {constant=<rgba 0 0 -500000 0>}) {desc=Move this below to avoid overdraw})
  ScalarParameter() {defaultValue=1, parameterName=MinOpacity, sliderMax=1}
  ScalarParameter() {defaultValue=5, parameterName=MaxOpacity, sliderMax=1}
  TransformPosition(None=Constant3Vector()) {transformType=TRANSFORMPOSSOURCE_TranslatedWorld}
  VectorParameter() {defaultValue=<rgba 3.9392 7.5419 -2.1513 -3.0102>}
  VertexInterpolator(VS=@n30)
 ]
}
