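# Niagara system for the Fizzy Cloud grenade's spray effect, built from one sprite emitter, Mist, scaled by User.FX_Scale.
# Spawn rate tapers over the system timeline, and initial velocity rotates around Z with system age, randomly choosing opposite sides for a spinning two-sided spray.
# Drag slows the flipbook-animated puffs as they grow, shift from bright HDR blue to dimmer tones, and fade with System.FadeOut.
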NiagaraSystem /FizzyCloudGameplay/Assets/FX/NS_FizzyCloud_GrenadeSpray
dependencies {
 usedDataInterfaces [
  &NiagaraDataInterfaceCurve
  &NiagaraDataInterfaceColorCurve
  &NiagaraDataInterfaceSpriteRendererInfo
 ]
 usedDynamicInputs [&Divide_Float, &FNTimeOfDayColor]
 usedModules [
  &SystemState
  &self:SystemSpawnScript.SystemScriptSource.SystemScriptGraph.NiagaraNodeAssignment_1.SetVariables_#1
  &self:SystemSpawnScript.SystemScriptSource.SystemScriptGraph.NiagaraNodeAssignment_2.SetVariables_#2
  &Timeline
 ]
}
emitters {Mist {
  emitterUpdateScript [
   &EmitterState()
   &SpawnRate(
    SpawnRate=&Multiply_Float(A=20, B=&OneMinusFloat(Float=$System.Timeline.NormalizedTime))
    Use Spawn Probability=false
   )
  ]
  particleSpawnScript [
   /SetVariables(Particles.SideAlpha:NiagaraFloat=&UniformAOrBFloat(A=0, Evaluation Type=::NewEnumerator0, Recalculate Random Each Loop=true, Override Seed=false)) {isSetParameters=true}
   &InitializeParticle(
    Color=&MultiplyLinearColor(
     Linear Color=&MultiplyLinearColor(
      Linear Color=&RandomRangeLinearColor(Minimum=<rgba 0.5521 0.5521 0.5521 1>, Maximum=<rgba 0.8 0.8 0.8 1>)
      Scale Factor=&MakeVector4_FromLinearColor(LinearColor=<rgba 1 1 1 0.487>)
     )
     Scale Factor=&MakeVector4_FromLinearColor(LinearColor=<rgba 1 1 1 1>)
    )
    Lifetime Max=2.4
    Lifetime Min=1.5
    Lifetime Mode=::NewEnumerator1
    Mass Mode=::NewEnumerator3
    Position Offset=&Multiply_VectorByFloat(Vector=<0 0 0>, Float=$User.FX_Scale)
    Sprite Rotation Mode=::NewEnumerator2
    Sprite Size Mode=::NewEnumerator3
    Uniform Sprite Size=&Multiply_Float(A=$User.FX_Scale, B=&RandomRangeFloat(Minimum=40, Maximum=50))
    UsePositionOffset=true
    Write Material Random=true
   )
   /SetVariables(Particles.InitialOffset:NiagaraFloat=&Multiply_Float(A=$System.Age, B=1)) {isSetParameters=true}
   &AddVelocity(
    Velocity=&RotateVectorByQuaternion(
     Vector To Rotate=&Lerp_Vector(A=<-1 0 0>, B=<1 0 0>, Alpha=$Particles.SideAlpha)
     Quaternion=&MakeQuatFromAxisAngle(Axis=<0 0 1>, AngleInDegrees=&Multiply_Float(A=$System.Age, B=100))
    )
    Velocity Speed Scale=&RandomRangeFloat(Minimum=50, Maximum=70)
   )
  ]
  particleUpdateScript [
   &ParticleState()
   &ScaleSpriteSize(Uniform Curve Sprite Scale=DataInterface() {propertyValues {Curve {keys [{interpMode=RCIM_Cubic, tangentMode=RCTM_User, time=0, value=0.4}, {interpMode=RCIM_Cubic, tangentMode=RCTM_User, time=0.7, value=1}]}}})
   &ScaleColor(
    Scale Alpha=&FloatFromCurve(
     FloatCurve=DataInterface() {propertyValues {Curve {keys [
         {arriveTangent=8.6057, interpMode=RCIM_Cubic, leaveTangent=8.6057, tangentMode=RCTM_User, time=0, value=0}
         {interpMode=RCIM_Cubic, tangentMode=RCTM_User, time=0.2, value=1}
         {interpMode=RCIM_Cubic, tangentMode=RCTM_User, time=1, value=0}
        ]}}}
     Scale Curve=&Multiply_Float(A=$System.FadeOut, B=1)
    )
    Scale RGB=&MakeVectorFromLinearColorRGB(Color=&ColorFromCurve(ColorCurve=DataInterface() {propertyValues {
        AlphaCurve {keys [<at 0 1 RCIM_Cubic>,<at 1 0 RCIM_Cubic>]}
        BlueCurve {keys [<at 0.1014 12 RCIM_Cubic>,<at 0.4835 2>,<at 0.7976 0.6771 RCIM_Cubic>]}
        GreenCurve {keys [<at 0.1014 1.2104 RCIM_Cubic>,<at 0.4835 0.22>,<at 0.7976 0.1623 RCIM_Cubic>]}
        RedCurve {keys [<at 0.1014 0 RCIM_Cubic>,<at 0.4835 0.3771>,<at 0.7976 0.4048 RCIM_Cubic>]}
       }}))
    ScaleA=true
    ScaleRGB=true
   )
   &SubUVAnimation(
    "Sprite Renderer "=DataInterface() {propertyValues {SpriteRenderer=&self:Mist_1.NiagaraSpriteRendererProperties_0}}
    UseEndFrame=false
    UseStartFrame=false
   )
   &Drag(Rotational Drag=2, Use Linear Drag=true, Use Rotational Drag=true)
   &SolveForcesAndVelocity(Clamp Velocity=false, Limit Acceleration=false)
   &DynamicMaterialParameters(Module.Index0Param0Write=true, Module.Index0Param1Write=true, Module.Index0Param2Write=true, Module.Index0Param3Write=true)
   &CameraOffset(Camera Offset Amount=50)
  ]
  rendererClasses [&NiagaraSpriteRendererProperties]
  renderers [{rendererClass=&NiagaraSpriteRendererProperties}]
 }}
summary {
 systemName=NS_FizzyCloud_GrenadeSpray
 emitters [{emitterName=Mist, rendererClasses [&NiagaraSpriteRendererProperties]}]
 userVariables [
  User.FadeOut {type=NiagaraBool}
  User.FX_Scale {type=NiagaraFloat, defaultValue=1, description=""}
  User.FXAge {type=NiagaraFloat, defaultValue=14, description=""}
  User.NightMultiplier {type=NiagaraFloat, defaultValue=1, description=""}
  User.SmokeColor {type=LinearColor, description="", defaultValue {struct=LinearColor, value=<rgba 0.5684 0.2959 0.7812 1>}}
 ]
}
