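# Niagara system for the Fizzy Cloud grenade's dome: a scalable, swirling purple smoke cloud with system-driven fade.
# Mist sprites orbit into a globe, faded by camera position and brightened by sky-light direction.
# WindGusts adds spinning, time-of-day-tinted meshes, while Screen holds one mesh, visibility-toggled by an age curve.
# Head particles circle under point attraction, and Flicker spawns from them as flickering colored sprites driven by curl noise and vortex forces.
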NiagaraSystem /FizzyCloudGameplay/Assets/FX/NS_FizzyCloud_Grenade_Dome
dependencies {
 usedDataInterfaces [
  &NiagaraDataInterfaceCurve
  &NiagaraDataInterfaceCamera
  &NiagaraDataInterfaceSpriteRendererInfo
  &NiagaraDataInterfaceArrayFloat
  &NiagaraDataInterfaceParticleRead
  &NiagaraDataInterfaceColorCurve
 ]
 usedDynamicInputs [&Divide_Float]
 usedModules [
  &SystemState
  &Timeline
  &self:SystemSpawnScript.SystemScriptSource.SystemScriptGraph.NiagaraNodeAssignment_1.SetVariables_#1
  &self:SystemSpawnScript.SystemScriptSource.SystemScriptGraph.NiagaraNodeAssignment_2.SetVariables_#2
  &self:StateControlFade
 ]
}
emitters {
 Flicker {
  emitterUpdateScript [
   &EmitterState(Life Cycle Mode=::NewEnumerator1, Loop Behavior=::NewEnumerator1, Loop Duration Mode=::NewEnumerator1, UseLoopDelay=false)
   /SetVariables(Emitter.SpawnPerFrame:NiagaraFloat=&RandomRangeFloat(Minimum=0, Maximum=60, Evaluation Type=::NewEnumerator1)) {enabled=false, isSetParameters=true}
   &SpawnRate(SpawnRate=$Emitter.SpawnPerFrame, Use Spawn Probability=false) {enabled=false, module=SpawnRate002}
   &SpawnRate(SpawnGroup=1, SpawnRate=$Emitter.SpawnPerFrame, Use Spawn Probability=false) {enabled=false, module=SpawnRate001}
   &SpawnRate(SpawnGroup=2, SpawnRate=$Emitter.SpawnPerFrame, Use Spawn Probability=false) {enabled=false}
   &SpawnBurst_Instantaneous(Spawn Count=3, Use Spawn Probability=false, UseLoopCountLimit=false) {enabled=false}
   &SpawnParticlesFromOtherEmitter(
    Attribute Reader=DataInterface() {propertyValues {EmitterBinding {bindingMode=Other, emitterName=Head}}}
    Spawn Rate=&RandomRangeFloat(Minimum=0, Maximum=90, Evaluation Type=::NewEnumerator1)
    Use Maximum Age=false
    Use Minimum Age=false
   )
  ]
  particleSpawnScript [
   &InitializeParticle(
    Color=<rgba 20 20 20 5>
    Lifetime=&RandomRangeFloat(Minimum=0.5, Maximum=6)
    Sprite Size Mode=::NewEnumerator3
    Uniform Sprite Size=&FloatFromCurve(
     FloatCurve=DataInterface() {propertyValues {Curve {keys [
         <at 0 1 RCIM_Cubic>
         {arriveTangent=-0.5457, interpMode=RCIM_Cubic, leaveTangent=-0.5457, tangentMode=RCTM_User, time=0.0942, value=0.3646}
         <at 1 0.1297 RCIM_Cubic>
        ]}}}
     CurveIndex=&RandomRangeFloat(Minimum=0, Maximum=1)
     Scale Curve=200
    )
    UsePositionOffset=false
    Write Material Random=true
   )
   &InheritVelocity(Inherited Velocity Speed Limit=200, Speed Limit=true, UseSpeedThreshold=false)
   &CurlNoiseForce(Bias Noise=false, Noise Frequency=2, Noise Strength=334.0741, Pan Noise=false) {module=CurlNoiseForce001}
   &ApplyInitialForces()
   &SampleParticlesFromOtherEmitter()
  ]
  particleUpdateScript [
   &ParticleState()
   &SpriteRotationRate(Rotation Rate=&RandomRangeFloat(Minimum=50, Maximum=200))
   &ScaleSpriteSize(
    Uniform Curve Scale=&Waveform(
     "Waveform [1]"=::NewEnumerator8
     "[1] Frequency"=&RandomRangeFloat(Minimum=60, Maximum=30)
     "[1] Amplitude Min/Max"=<-1 1>
    )
    Uniform Curve Sprite Scale=DataInterface() {propertyValues {Curve {
       defaultValue=1
       keys [
        {arriveTangent=9.0725, interpMode=RCIM_Cubic, leaveTangent=9.0725, tangentMode=RCTM_User, time=0, value=0}
        {arriveTangent=-2.5249, interpMode=RCIM_Cubic, leaveTangent=-2.5249, tangentMode=RCTM_User, time=0.0445, value=0.9847}
        {arriveTangent=-1.0306, interpMode=RCIM_Cubic, leaveTangent=-1.0306, time=0.1676, value=0.3054}
        {arriveTangent=-0.0405, interpMode=RCIM_Cubic, leaveTangent=-0.0405, tangentMode=RCTM_User, time=1, value=0}
       ]
      }}}
   )
   &DynamicMaterialParameters(
    Index 0 Param 4=0.01
    Module.Index0Float0Write=true
    Module.Index0Float1Write=true
    Module.Index0Float2Write=true
    Module.Index0Float3Write=true
   )
   &CameraOffset(Camera Offset Amount=50)
   &ScaleColor(
    Color Scale:LinearColor=&ColorFromCurve(ColorCurve=DataInterface() {propertyValues {
       AlphaCurve {keys [<at 0 1 RCIM_Cubic>,<at 1 0 RCIM_Cubic>]}
       BlueCurve {keys [<at 0.0647 1 RCIM_Cubic>, {interpMode=RCIM_Cubic, leaveTangent=-0.7944, time=0.1301, value=1}, <at 0.4582 0.276>]}
       GreenCurve {keys [<at 0.0647 0.1926 RCIM_Cubic>, {interpMode=RCIM_Cubic, leaveTangent=-0.1166, time=0.1301, value=0.0856}, <at 0.4582 0.0863>]}
       RedCurve {keys [<at 0.0647 0.086 RCIM_Cubic>, {interpMode=RCIM_Cubic, leaveTangent=0.0924, time=0.1301, value=0.3404}, <at 0.4582 0.1702>]}
      }})
    Scale Mode=3
   )
   &CurlNoiseForce(Bias Noise=false, Noise Strength=500, Pan Noise=false)
   &PointForce(Force Strength=-500, Module.Use Falloff=false, Module.UseLinearForce=false) {enabled=false}
   &Drag(Drag=5, Use Linear Drag=true, Use Rotational Drag=true)
   &VortexForce(UseOriginOffset=false, Vortex Force Amount=1000)
   &GravityForce(Gravity=<0 0 980>)
   &SolveForcesAndVelocity(Clamp Velocity=false, Limit Acceleration=false)
  ]
  rendererClasses [&NiagaraSpriteRendererProperties]
  renderers [{rendererClass=&NiagaraSpriteRendererProperties}]
 }
 Head {
  emitterUpdateScript [
   &EmitterState()
   &SpawnRate(Use Spawn Probability=false)
  ]
  particleSpawnScript [
   &InitializeParticle(
    Color Mode=::NewEnumerator0
    Lifetime=&RandomRangeFloat(Minimum=2, Maximum=4)
    UsePositionOffset=false
    Write Material Random=true
   )
   &ShapeLocation(
    MaskSpawnGroup=false
    Override Seed=false
    Ring Radius=400
    Rotation Mode=::NewEnumerator2
    Shape Primitive=::NewEnumerator4
    UseOwnerScale=true
    Yaw / Pitch / Roll=&RandomVector(Vector Scale=0.1)
   )
   &VortexForce(UseOriginOffset=false) {enabled=false}
   &ApplyInitialForces() {enabled=false}
   &AddVelocity(Velocity=$StackContext.ShapeLocation.ShapeBitangent, Velocity Speed Scale=-1000)
  ]
  particleUpdateScript [
   &ParticleState()
   &PointForce(Force Strength=-4000, Module.Use Falloff=false, Module.UseLinearForce=false)
   &SolveForcesAndVelocity(Clamp Velocity=false, Limit Acceleration=false)
  ]
  rendererClasses [&NiagaraSpriteRendererProperties]
  renderers [{rendererClass=&NiagaraSpriteRendererProperties}]
 }
 Mist {
  emitterUpdateScript [
   &EmitterState()
   &SpawnRate(SpawnRate=12, Use Spawn Probability=false)
  ]
  particleSpawnScript [
   /SetVariables(Particles.SmokeGlobeAlpha:NiagaraFloat=&FloatFromCurve(
     FloatCurve=DataInterface() {propertyValues {Curve {keys [
         {arriveTangent=0.3467, interpMode=RCIM_Cubic, leaveTangent=0.3467, tangentMode=RCTM_User, time=0, value=0}
         {arriveTangent=0.4113, interpMode=RCIM_Cubic, leaveTangent=0.4113, tangentMode=RCTM_User, time=1, value=0.2}
        ]}}}
     CurveIndex=&RandomRangeFloat(Minimum=0, Maximum=1)
    )) {isSetParameters=true}
   &InitializeParticle(
    Color=&MultiplyLinearColor(
     Linear Color=&MultiplyLinearColor(
      Linear Color=&RandomRangeLinearColor(Minimum=<rgba 0.5521 0.5521 0.5521 1>, Maximum=<rgba 0.8 0.8 0.8 1>)
      Scale Factor=<1 1 1 1>
     )
     Scale Factor=&MakeVector4_FromLinearColor(LinearColor=<rgba 0.3374 0.2335 0.6771 1>)
    )
    Lifetime Max=1.8
    Lifetime Mode=::NewEnumerator1
    Mass Mode=::NewEnumerator3
    Position Offset=&Multiply_VectorByFloat(Vector=<0 0 0>, Float=$User.FX_Scale)
    Sprite Rotation Mode=::NewEnumerator2
    Sprite Size Mode=::NewEnumerator3
    Uniform Sprite Size=&Multiply_Float(A=$User.FX_Scale, B=&RandomRangeFloat(Minimum=300, Maximum=350))
    UsePositionOffset=true
    Write Material Random=true
   )
   /SetVariables(Particles.InitialOffset:NiagaraFloat=&Multiply_Float(A=$System.Age, B=1)) {isSetParameters=true}
   &CameraOffset(Camera Offset Amount=&Multiply_Float(A=60, B=$User.FX_Scale))
  ]
  particleUpdateScript [
   &ParticleState()
   &RotateAroundPoint(
    Radius:NiagaraFloat=&Multiply_Float(
     A=&Lerp_Float(
      A=350
      B=1
      Alpha=&FloatFromCurve(
       FloatCurve=DataInterface() {propertyValues {Curve {keys [
           {arriveTangent=0.1398, interpMode=RCIM_Cubic, leaveTangent=0.1398, tangentMode=RCTM_User, time=0, value=0}
           {arriveTangent=3.3869, interpMode=RCIM_Cubic, leaveTangent=3.3869, tangentMode=RCTM_User, time=1, value=1}
          ]}}}
       CurveIndex=$Particles.SmokeGlobeAlpha
      )
     )
     B=$User.FX_Scale
    )
    Rotation Center:NiagaraPosition=&Multiply_VectorByFloat(
     Vector=&Lerp_Vector(
      A=<0 0 0>
      B=<0 0 350>
      Alpha=&Multiply_Float(
       A=$Particles.SmokeGlobeAlpha
       B=&UniformAOrBFloat(A=-1, Evaluation Type=::NewEnumerator0, Recalculate Random Each Loop=true, Override Seed=false)
      )
     )
     Float=$User.FX_Scale
    )
    Rotation Phase:NiagaraFloat=&Add_Float(
     A=&FloatFromCurve(
      FloatCurve=DataInterface() {propertyValues {Curve {keys [<at 0 1>,<at 1 0>]}}}
      Scale Curve=&RandomRangeFloat(Minimum=-0.1, Maximum=-0.1)
     )
     B=$Particles.InitialOffset
    )
   )
   &self:FadeOutBackParticles(
    CamInsideFallof=100
    CamInsideRadius=600
    CameraForwardVector:Vector3f=&ReturnCameraVector(Camera Vectors=::NewEnumerator1, Coordinate Space=::World)
    CameraPosition:NiagaraPosition=&ReturnCameraPosition(Coordinate Space:ENiagaraCoordinateSpace=::World, Camera Query:NiagaraDataInterfaceCamera=DataInterface())
    FadeOutSoftness=0.15
    FadeOutTreshold=0.5
   ) {module=FadeOutBackParticles}
   &self:FadeOutWhenInSmoke(AlphaMax=1, AlphaMin=0.5, FadeFallof=100, FadeRadius=600, SmokePos:NiagaraPosition=$Engine.Owner.Position) {module=FadeOutWhenInSmoke}
   &ScaleSpriteSize(Uniform Curve Sprite Scale=DataInterface() {propertyValues {Curve {keys [{interpMode=RCIM_Cubic, tangentMode=RCTM_User, time=0, value=0.6}, {interpMode=RCIM_Cubic, tangentMode=RCTM_User, time=0.7, value=1}]}}})
   &ScaleColor(
    Scale Alpha=&FloatFromCurve(
     FloatCurve=DataInterface() {propertyValues {Curve {keys [
         {interpMode=RCIM_Cubic, tangentMode=RCTM_User, time=0, value=0}
         {interpMode=RCIM_Cubic, tangentMode=RCTM_User, time=0.4447, value=0.95}
         {interpMode=RCIM_Cubic, tangentMode=RCTM_User, time=1, value=0}
        ]}}}
     Scale Curve=&Multiply_Float(
      A=&Multiply_Float(
       A=&Multiply_Float(A=$System.FadeOut, B=&OneMinusFloat(Float=$System.Timeline.NormalizedTime))
       B=$Output.FadeOutBackParticles.FresnelAlpha
      )
      B=$Output.FadeOutWhenInSmoke.FadeOutAlpha
     )
    )
    Scale RGB=&Multiply_Vector(
     A=&MakeVectorFromLinearColorRGB(Color=$System.TODColorMult)
     B=&VectorFromFloat(Value=&Multiply_Float(
       A=$User.NightMultiplier
       B=&Lerp_Float(
        A=1
        B=1.5
        Alpha=&FloatFromCurve(
         FloatCurve=DataInterface() {propertyValues {Curve {keys [<at -0.75 0>,<at 1 1>]}}}
         CurveIndex=&DotProduct(
          Vector A=&NormalizeVector(
           Vector To Normalize=&MakeVectorFromLinearColorRGB(Color=$NPC.FortniteNPC.SkyAtmosphereLightDirection)
           Safe Normalize=true
           Safe Fallback Threshold=0
           Safe Fallback Vector=<0 0 0.001>
          )
          Vector B=&TransformVector(
           Vector=&NormalizeVector(
            Vector To Normalize=$Particles.RotateAroundPoint.VectorFromCenter
            Safe Normalize=true
            Safe Fallback Threshold=0
            Safe Fallback Vector=<0 0 0.001>
           )
           Destination Space=::World
          )
         )
         Scale Curve=5
        )
       )
      ))
    )
    ScaleA=true
    ScaleRGB=true
   )
   &SubUVAnimation(
    "Sprite Renderer "=DataInterface() {propertyValues {SpriteRenderer=&self:Mist_1.NiagaraSpriteRendererProperties_0}}
    UseEndFrame=false
    UseStartFrame=false
   )
   &Drag(Drag=4, Rotational Drag=2, Use Linear Drag=true, Use Rotational Drag=true)
   &SolveForcesAndVelocity(Clamp Velocity=false, Limit Acceleration=false)
   &DynamicMaterialParameters(Module.Index0Param0Write=true, Module.Index0Param1Write=true, Module.Index0Param2Write=true, Module.Index0Param3Write=true)
   &ScaleSpriteSize(Uniform Curve Index=$Output.FadeOutBackParticles.FresnelAlpha) {module=ScaleSpriteSize001}
   &self:FuckPerf(Input Max=900, Input Min=800) {module=FuckPerf}
  ]
  rendererClasses [&NiagaraSpriteRendererProperties]
  renderers [{rendererClass=&NiagaraSpriteRendererProperties}]
 }
 Screen {
  emitterUpdateScript [
   &EmitterState()
   &SpawnBurst_Instantaneous(Use Spawn Probability=false, UseLoopCountLimit=false)
   /SetVariables(Emitter.FadeValue:NiagaraFloat=&FloatFromCurve(
     FloatCurve=DataInterface() {propertyValues {Curve {keys [
         <at 0 0 RCIM_Cubic>
         {arriveTangentWeight=0.0945, interpMode=RCIM_Cubic, leaveTangentWeight=1.5307, tangentMode=RCTM_User, time=0.5, value=1}
         {arriveTangentWeight=1.5314, interpMode=RCIM_Cubic, leaveTangentWeight=0.0776, tangentMode=RCTM_User, time=11.5, value=1}
         <at 12 0 RCIM_Cubic>
        ]}}}
     CurveIndex=$System.Age
     Scale Curve=$System.Opacity
    )) {isSetParameters=true}
  ]
  particleSpawnScript [&InitializeParticle(Color Mode=::NewEnumerator0, Lifetime=12, UsePositionOffset=false, Write Material Random=true)]
  particleUpdateScript [
   &ParticleState()
   /SetVariables(Particles.VisibilityTag=&MakeCustomIntFromBool(Bool=&SetBoolByFloatComparison(A=$Emitter.FadeValue, B=0), True INT=0, False INT=1)) {isSetParameters=true}
  ]
  rendererClasses [&NiagaraMeshRendererProperties]
  renderers [{rendererClass=&NiagaraMeshRendererProperties, rendererIndex=0}]
 }
 WindGusts {
  emitterUpdateScript [
   &EmitterState()
   &SpawnRate(
    Spawn Probability=&FloatFromCurve(
     FloatCurve=DataInterface() {propertyValues {Curve {keys [
         {interpMode=RCIM_Cubic, tangentMode=RCTM_User, time=0, value=0}
         {interpMode=RCIM_Cubic, tangentMode=RCTM_User, time=0.1, value=1}
         {arriveTangent=-4.1541, interpMode=RCIM_Cubic, leaveTangent=-4.1541, tangentMode=RCTM_User, time=1, value=0}
        ]}}}
     CurveIndex=$System.NormalizedLoopAge
    )
    SpawnRate=6
    Use Spawn Probability=false
   )
  ]
  particleSpawnScript [
   /SetVariables(Particles.Particle_Height:NiagaraFloat=&RandomRangeFloat(Minimum=0.4, Maximum=0.6)) {isSetParameters=true}
   &InitializeParticle(
    Color=&MultiplyLinearColor(
     Linear Color=&MultiplyLinearColor(
      Linear Color=&RandomRangeLinearColor(Minimum=<rgba 0.8958 0.8958 0.8958 1>, Maximum=<rgba 1.2292 1.2292 1.2292 1>)
      Scale Factor=<1 1 1 1>
     )
     Scale Factor=&MakeVector4_FromLinearColor(LinearColor=<rgba 0.6495 0.3333 1 1>)
    )
    Lifetime Max=0.8
    Lifetime Min=0.6
    Lifetime Mode=::NewEnumerator1
    Mass Mode=::NewEnumerator3
    Mesh Scale=&Multiply_VectorByFloat(
     Vector=&RandomRangeVector(Minimum=<1 1 0.5>, Maximum=<1 1 1.5>)
     Float=&Multiply_Float(
      A=$User.FX_Scale
      B=&Lerp_Float(
       A=5.5
       B=6
       Alpha=&FloatFromCurve(
        FloatCurve=DataInterface() {propertyValues {Curve {keys [
            {arriveTangent=5.7133, interpMode=RCIM_Cubic, leaveTangent=5.7133, tangentMode=RCTM_User, time=0, value=0}
            {interpMode=RCIM_Cubic, tangentMode=RCTM_User, time=0.5, value=1}
            {arriveTangent=-6.0152, interpMode=RCIM_Cubic, leaveTangent=-6.0152, tangentMode=RCTM_User, time=1, value=0}
           ]}}}
        CurveIndex=$Particles.Particle_Height
        Scale Curve=&RandomRangeFloat(Minimum=0.9, Maximum=1)
       )
      )
     )
    )
    Mesh Scale Mode=::NewEnumerator4
    Sprite Rotation Mode=::NewEnumerator2
    UsePositionOffset=false
    Write Material Random=true
   )
   &SystemLocation(
    MaskSpawnGroup=false
    Offset=&MakeVector(
     X=0
     Y=0
     Z=&Multiply_Float(A=$User.FX_Scale, B=&Lerp_Float(A=-300, B=300, Alpha=$Particles.Particle_Height))
    )
   )
   &InitialMeshOrientation(
    Mesh Orientation Mode=::NewEnumerator2
    Orientation Coordinate Space=::World
    Rotation=&RandomRangeVector(Minimum=<-0.04 -0.04 -1>, Maximum=<0 0 1>)
    Use Rotation Vector=true
   )
   &CameraOffset(Camera Offset Amount=15)
  ]
  particleUpdateScript [
   &ParticleState()
   &ScaleMeshSize(Scale Factor=&Multiply_VectorByFloat(
     Vector=&MakeVector(
      X=1
      Y=1
      Z=&SelectFloatFromArray(
       Array Sampling Mode=::NewEnumerator0
       Direct Array Index=?
       Float Selection Array=DataInterface() {propertyValues {FloatData [1,-1]}}
       Random Seed=0
      )
     )
     Float=&Lerp_Float(
      A=0.9
      B=1.1
      Alpha=&FloatFromCurve(FloatCurve=DataInterface() {propertyValues {Curve {keys [
           {arriveTangent=2.1221, interpMode=RCIM_Cubic, leaveTangent=2.1221, tangentMode=RCTM_User, time=0, value=0}
           {interpMode=RCIM_Cubic, tangentMode=RCTM_User, time=0.86, value=1}
          ]}}})
     )
    ))
   &UpdateMeshOrientation(Rotation Rate=&RandomRangeFloat(Minimum=0.1, Maximum=0.3), Rotation Vector=<0 0 1>)
   &ScaleColor(
    Scale Alpha=&FloatFromCurve(
     FloatCurve=DataInterface() {propertyValues {Curve {keys [
         {interpMode=RCIM_Cubic, leaveTangentWeight=1.0516, tangentMode=RCTM_User, time=0, value=0}
         {arriveTangentWeight=1.164, interpMode=RCIM_Cubic, leaveTangentWeight=1.164, tangentMode=RCTM_User, time=0.2085, value=0.8824}
         {arriveTangent=-1.1189, interpMode=RCIM_Cubic, leaveTangent=-1.1189, time=0.8864, value=0.8971}
         {arriveTangent=0.4048, interpMode=RCIM_Cubic, leaveTangent=0.4048, tangentMode=RCTM_User, time=0.9971, value=0}
        ]}}}
     Scale Curve=&Multiply_Float(
      A=&RandomRangeFloat(Minimum=3, Maximum=3)
      B=&Multiply_Float(A=&Multiply_Float(A=$System.FadeOut, B=1), B=1)
     )
    )
    Scale RGB=&Multiply_Vector(
     A=&Multiply_VectorByFloat(
      Vector=&MakeVectorFromLinearColorRGB(Color=$System.TODColorMult)
      Float=&Multiply_Float(
       A=&Lerp_Float(
        A=0.5
        B=0.75
        Alpha=&FloatFromCurve(FloatCurve=DataInterface() {propertyValues {Curve {keys [
             {interpMode=RCIM_Cubic, tangentMode=RCTM_User, time=-0.0105, value=0.7647}
             {interpMode=RCIM_Cubic, tangentMode=RCTM_User, time=0.4982, value=0.1882}
             {interpMode=RCIM_Cubic, tangentMode=RCTM_User, time=0.9825, value=0.9176}
            ]}}})
       )
       B=$User.NightMultiplier
      )
     )
     B=&MakeVectorFromLinearColorRGB(Color=&FNTimeOfDayColor(
       Night Color=<rgba 0.4274 0.8859 1.3 1>
       Morning Color=<rgba 0.672 0.7134 1 1>
       Day Color=<rgba 0.7135 0.9024 1 1>
       Evening Color=<rgba 0.6957 0.5125 0.8 1>
      ))
    )
    ScaleA=true
    ScaleRGB=true
   )
   &SolveForcesAndVelocity(Clamp Velocity=false, Limit Acceleration=false)
   &DynamicMaterialParameters(
    Index 0 Param 1=&FloatFromCurve(FloatCurve=DataInterface() {propertyValues {Curve {keys [
         {arriveTangent=-0.0245, interpMode=RCIM_Cubic, leaveTangent=-0.0245, tangentMode=RCTM_User, time=-0.0036, value=0.5195}
         {arriveTangent=0.0772, interpMode=RCIM_Cubic, leaveTangent=0.0772, tangentMode=RCTM_User, time=0.5011, value=0.3103}
         {interpMode=RCIM_Cubic, tangentMode=RCTM_User, time=1.0107, value=0.5658}
        ]}}})
    Index 0 Param 3=1
    Module.Index0Param0Write=true
    Module.Index0Param1Write=true
    Module.Index0Param2Write=true
    Module.Index0Param3Write=true
   )
  ]
  rendererClasses [&NiagaraMeshRendererProperties]
  renderers [{rendererClass=&NiagaraMeshRendererProperties, rendererIndex=0}]
 }
}
summary {
 systemName=NS_FizzyCloud_Grenade_Dome
 emitters [
  {emitterName=Mist, rendererClasses [&NiagaraSpriteRendererProperties]}
  {emitterName=WindGusts, rendererClasses [&NiagaraMeshRendererProperties]}
  {emitterName=Flicker, rendererClasses [&NiagaraSpriteRendererProperties]}
  {emitterName=Head, rendererClasses [&NiagaraSpriteRendererProperties]}
  {emitterName=Screen, rendererClasses [&NiagaraMeshRendererProperties]}
 ]
 userVariables [
  User.FadeOut {type=NiagaraBool}
  User.FX_Scale {type=NiagaraFloat, defaultValue=1.2, description=""}
  User.FXAge {type=NiagaraFloat, defaultValue=20, description=""}
  User.NightMultiplier {type=NiagaraFloat, defaultValue=1, description=""}
  User.SmokeColor {type=LinearColor, description="", defaultValue {struct=LinearColor, value=<rgba 0.9049 0.8594 1 1>}}
 ]
}
