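# Niagara system for a FizzyCloud projectile, pairing a glowing head with a smoke trail.
# Trail spawns two interleaved ribbons via ParticleID modulo, pinning new points to the emitter before they release, widen and cool from bright blue to dim smoke.
# Flare bursts persistent, camera-offset blue-violet sprites with array-picked shapes, slow spin and pulsing size.
# Flicker spawns short-lived sparks in three shape groups at randomized rates, moved by curl noise, negative point force and drag.
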NiagaraSystem /FizzyCloudGameplay/Assets/FX/NS_FizzyCloud_ProjectileSmokeTrail
dependencies {
 usedDataInterfaces [
  &NiagaraDataInterfaceColorCurve
  &NiagaraDataInterfaceCurve
  &NiagaraDataInterfaceArrayFloat2
  &NiagaraDataInterfaceArrayFloat
 ]
 usedDynamicInputs [&SetBoolByFloatComparison, &VectorLength]
 usedModules [&SystemState, &FN_Timer]
}
emitters {
 Flare {
  emitterUpdateScript [
   &EmitterState(Life Cycle Mode=::NewEnumerator1, Loop Behavior=::NewEnumerator1, Loop Duration=1, UseLoopDelay=false)
   &SpawnBurst_Instantaneous(Spawn Count=5, Use Spawn Probability=false, UseLoopCountLimit=false)
  ]
  particleSpawnScript [&InitializeParticle(
    Color=&RandomRangeLinearColor(Minimum=<rgba 0.1068 0.6772 1 1>, Maximum=<rgba 0.0845 0 1 1>, Randomness Mode=::NewEnumerator1, Random Seed=3)
    Sprite Rotation Angle=&SelectFloatFromArray(
     Direct Array Index=&ReturnExecIndex()
     Float Selection Array=DataInterface() {propertyValues {FloatData [0,45,-45,0]}}
    )
    Sprite Rotation Mode=::NewEnumerator0
    Sprite Size=&Multiply_Vector2D(
     A=&SelectVector2DFromArray(
      Direct Array Index=&ReturnExecIndex()
      Vector 2D Selection Array=DataInterface() {propertyValues {FloatData [<100 400>, <200 10>, <10 200>, <400 400>, <1000 10>]}}
     )
     B=<1 1>
    )
    Sprite Size Mode=::NewEnumerator4
    UsePositionOffset=false
    Write Material Random=true
   )]
  particleUpdateScript [
   &ParticleState(Kill Particles When Lifetime Has Elapsed=false)
   &SpriteRotationRate(
    Delta Time:NiagaraFloat=&Multiply_Float(
     A=&Sine(
      Normalized Angle=$Engine.Time
      Period=5.8107
      Scale=0.854
      Bias=&RandomRangeFloat(Minimum=0, Maximum=1)
     )
     B=&RandomRangeFloat(Minimum=0.2, Maximum=0.1)
    )
    Rotation Rate=&RandomRangeFloat(Minimum=25, Maximum=50)
   )
   &ScaleSpriteSize(Uniform Curve Scale=&Waveform("Waveform [1]"=::NewEnumerator8, "[1] Frequency"=45, "[1] Amplitude Min/Max"=<0.7 1>))
   &DynamicMaterialParameters(
    Index 0 Param 4=0.1
    Module.Index0Float0Write=true
    Module.Index0Float1Write=true
    Module.Index0Float2Write=true
    Module.Index0Float3Write=true
   )
   &CameraOffset(Camera Offset Amount=50)
   &ScaleColor(Scale Mode=3)
  ]
  rendererClasses [&NiagaraSpriteRendererProperties]
  renderers [{rendererClass=&NiagaraSpriteRendererProperties}]
 }
 Flicker {
  emitterUpdateScript [
   &EmitterState(Life Cycle Mode=::NewEnumerator1, Loop Behavior=::NewEnumerator1, Loop Duration Mode=::NewEnumerator1, UseLoopDelay=false)
   /SetVariables(Emitter.SpawnPerFrame:NiagaraFloat=&RandomRangeFloat(Minimum=0, Maximum=60, Evaluation Type=::NewEnumerator1)) {isSetParameters=true}
   &SpawnRate(SpawnRate=$Emitter.SpawnPerFrame, Use Spawn Probability=false) {module=SpawnRate002}
   &SpawnRate(SpawnGroup=1, SpawnRate=$Emitter.SpawnPerFrame, Use Spawn Probability=false) {module=SpawnRate001}
   &SpawnRate(SpawnGroup=2, SpawnRate=$Emitter.SpawnPerFrame, Use Spawn Probability=false)
   &SpawnBurst_Instantaneous(Spawn Count=3, Use Spawn Probability=false, UseLoopCountLimit=false) {enabled=false}
  ]
  particleSpawnScript [
   &InitializeParticle(
    Color=&RandomRangeLinearColor(Minimum=<rgba 0.8063 0.0891 1 1>, Maximum=<rgba 0 0.4512 1 1>, Randomness Mode=::NewEnumerator1, Random Seed=3)
    Lifetime=&RandomRangeFloat(Minimum=0.5, Maximum=1)
    Sprite Rotation Angle=&SelectFloatFromArray(
     Direct Array Index=$Emitter.SpawnGroup
     Direct Array Mode=::NewEnumerator1
     Float Selection Array=DataInterface() {propertyValues {FloatData [0,45,45]}}
    )
    Sprite Rotation Mode=::NewEnumerator0
    Sprite Size=&Multiply_Vector2D(
     A=&SelectVector2DFromArray(
      Direct Array Index=$Emitter.SpawnGroup
      Direct Array Mode=::NewEnumerator1
      Vector 2D Selection Array=DataInterface() {propertyValues {FloatData [<100 100>, <100 10>, <10 100>]}}
     )
     B=<1 1>
    )
    Sprite Size Mode=::NewEnumerator4
    UsePositionOffset=false
    Write Material Random=true
   )
   &InheritVelocity(Inherited Velocity Speed Limit=200, Speed Limit=true, UseSpeedThreshold=false)
   &CurlNoiseForce(Bias Noise=false, Noise Frequency=2, Noise Strength=334.0741, Pan Noise=false) {module=CurlNoiseForce001}
   &ApplyInitialForces()
  ]
  particleUpdateScript [
   &ParticleState()
   &SpriteRotationRate(Rotation Rate=&RandomRangeFloat(Minimum=50, Maximum=200))
   &ScaleSpriteSize(
    Uniform Curve Scale=&Waveform(
     "Waveform [1]"=::NewEnumerator8
     "[1] Frequency"=&RandomRangeFloat(Minimum=60, Maximum=30)
     "[1] Amplitude Min/Max"=<-1 1>
    )
    Uniform Curve Sprite Scale=DataInterface() {propertyValues {Curve {
       defaultValue=1
       keys [
        {arriveTangent=9.0725, interpMode=RCIM_Cubic, leaveTangent=9.0725, tangentMode=RCTM_User, time=0, value=0}
        {arriveTangent=-2.5249, interpMode=RCIM_Cubic, leaveTangent=-2.5249, tangentMode=RCTM_User, time=0.1408, value=0.9847}
        {arriveTangent=-0.0405, interpMode=RCIM_Cubic, leaveTangent=-0.0405, tangentMode=RCTM_User, time=1, value=0}
       ]
      }}}
   )
   &DynamicMaterialParameters(
    Index 0 Param 4=0.01
    Module.Index0Float0Write=true
    Module.Index0Float1Write=true
    Module.Index0Float2Write=true
    Module.Index0Float3Write=true
   )
   &CameraOffset(Camera Offset Amount=50)
   &ScaleColor(Color Scale:LinearColor=<rgba 8 8 8 10>, Scale Mode=3)
   &CurlNoiseForce(Bias Noise=false, Noise Strength=500, Pan Noise=false)
   &PointForce(Force Strength=-500, Module.Use Falloff=false, Module.UseLinearForce=false)
   &Drag(Drag=5, Use Linear Drag=true, Use Rotational Drag=true)
   &SolveForcesAndVelocity(Clamp Velocity=false, Limit Acceleration=false)
  ]
  rendererClasses [&NiagaraSpriteRendererProperties]
  renderers [{rendererClass=&NiagaraSpriteRendererProperties}]
 }
 Trail {
  emitterUpdateScript [
   &EmitterState(Delay First Loop Only=true, Inactive Response=::NewEnumerator1, Life Cycle Mode=::NewEnumerator1, Loop Duration=1, UseLoopDelay=true)
   &SpawnRate(SpawnRate=20, Use Spawn Probability=false)
  ]
  particleSpawnScript [
   &InitializeParticle(
    Color=&RandomRangeLinearColor(Minimum=<rgba 0.0885 0.0885 0.0885 0.5>, Maximum=<rgba 1 1 1 1>)
    Ribbon Width=&RandomRangeFloat(Minimum=5, Maximum=60)
    Ribbon Width Mode=::NewEnumerator1
    UsePositionOffset=false
    Write Material Random=true
   )
   &AddVelocity(Module.Use Falloff Distance=false, Override Seed=false, Velocity Mode=::NewEnumerator1, Velocity Speed=50)
   /SetVariables(Particles.RibbonID=/Niagara/DynamicInputs/NiagaraID/MakeNiagaraIDFromInt(
     Index:NiagaraInt32=&ModuloInt(A=&ReturnParticleIDAttributes(), B=2)
     Acquire Tag:NiagaraInt32=0
    )) {isSetParameters=true}
  ]
  particleUpdateScript [
   &ParticleState()
   &SolveForcesAndVelocity(Clamp Velocity=false, Limit Acceleration=false)
   &ScaleColor(
    Color Scale:LinearColor=&ColorFromCurve(ColorCurve=DataInterface() {propertyValues {
       AlphaCurve {keys [<at 0 1 RCIM_Cubic>,<at 0.1502 0.5>,<at 1 0 RCIM_Cubic>]}
       BlueCurve {keys [
         <at 0.0156 10>
         {arriveTangent=-35.8851, interpMode=RCIM_Cubic, leaveTangent=-35.8851, time=0.1527, value=2}
         {arriveTangent=-2.2193, interpMode=RCIM_Cubic, leaveTangent=-2.2193, time=0.2736, value=0.7396}
         <at 0.9718 0.1823>
        ]}
       GreenCurve {keys [
         <at 0.0156 1.2722>
         {arriveTangent=-4.5815, interpMode=RCIM_Cubic, leaveTangent=-4.5815, time=0.1527, value=0.2544}
         {arriveTangent=-0.2902, interpMode=RCIM_Cubic, leaveTangent=-0.2902, time=0.2736, value=0.0899}
         <at 0.9718 0.0167>
        ]}
       RedCurve {keys [
         <at 0.0156 0.0204>
         {arriveTangent=0.5726, interpMode=RCIM_Cubic, leaveTangent=0.5726, time=0.1527, value=0.0041}
         {arriveTangent=0.186, interpMode=RCIM_Cubic, leaveTangent=0.186, time=0.2736, value=0.1682}
         <at 0.9718 0.1564>
        ]}
      }})
    Scale Mode=3
   )
   /SetVariables(Particles.Position=&Lerp_Vector(
     A=Linked() {linkedVariable {name=Engine.Emitter.SimulationPosition, type=NiagaraPosition}}
     B=Linked() {linkedVariable {name=Particles.Position, type=NiagaraPosition}}
     Alpha=&FloatFromCurve(FloatCurve=DataInterface() {propertyValues {Curve {
         defaultValue=1
         keys [{interpMode=RCIM_Cubic, tangentMode=RCTM_User, time=0, value=0}, {interpMode=RCIM_Cubic, tangentMode=RCTM_User, time=0.1011, value=1}]
        }}})
    )) {isSetParameters=true}
   &ScaleRibbonWidth(Ribbon Width Scale=&FloatFromCurve(FloatCurve=DataInterface() {propertyValues {Curve {
        defaultValue=1
        keys [{interpMode=RCIM_Cubic, tangentMode=RCTM_User, time=0, value=0.9971}, {interpMode=RCIM_Cubic, tangentMode=RCTM_User, time=1, value=1.9971}]
       }}}))
  ]
  rendererClasses [&NiagaraRibbonRendererProperties]
  renderers [{rendererClass=&NiagaraRibbonRendererProperties}]
 }
}
summary {
 systemName=NS_FizzyCloud_ProjectileSmokeTrail
 emitters [
  {emitterName=Trail, rendererClasses [&NiagaraRibbonRendererProperties]}
  {emitterName=Flare, rendererClasses [&NiagaraSpriteRendererProperties]}
  {emitterName=Flicker, rendererClasses [&NiagaraSpriteRendererProperties]}
 ]
 userVariables []
}
